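# Niagara system for the WidePlane pickaxe's trail, built from a single short-lived ribbon emitter, MainTrail.
# FN anim-trail modules spawn, initialize and update particles from the system's skeletal mesh, with initial color set by time of day.
# Over particle life the ribbon narrows, darkens slightly and fades out.
# A second alpha curve adds a head fade-in, blended in by emitter age so it applies only after the first moments of emission.
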NiagaraSystem /BRCosmetics/Weapons/FORT_Melee/Pickaxe_WidePlane/FX/NS_Pickaxe_WidePlane_Trail
dependencies {
 usedDataInterfaces [&NiagaraDataInterfaceColorCurve, &NiagaraDataInterfaceCurve, &NiagaraDataInterfaceSkeletalMesh]
 usedDynamicInputs []
 usedModules [
  &self:SystemSpawnScript.SystemScriptSource.SystemScriptGraph.NiagaraNodeAssignment_0.SetVariables_#1
  &SystemState
 ]
}
emitters {MainTrail {
  emitterUpdateScript [
   &EmitterState()
   &FN_SpawnAnimTrail(Spawn Interval=0)
  ]
  particleSpawnScript [
   &FN_InitializeAnimTrailBoneIndecies()
   &FN_InitializeAnimTrail(
    Color=&FNTimeOfDayColor(
     Night Color=<rgba 0.125 0.0822 0.0608 6>
     Morning Color=<rgba 0.4219 0.1562 0.0898 15>
     Day Color=<rgba 0.2188 0.1097 0.0466 10>
     Evening Color=<rgba 0.4219 0.1562 0.0898 15>
    )
    Lifetime=0.12
    SkeletalMesh=$System.Skeletal Mesh
   )
  ]
  particleUpdateScript [
   &FN_UpdateAnimTrail(Locked Particle RibbonWidth=1, SkeletalMesh=$System.Skeletal Mesh) {module=FN_UpdateAnimTrail001}
   &ScaleColor(
    Scale Alpha=&FloatFromCurve(FloatCurve=DataInterface() {propertyValues {Curve {keys [
         {arriveTangent=-2.003, interpMode=RCIM_Cubic, leaveTangent=-2.003, tangentMode=RCTM_User, time=0.3, value=1}
         {interpMode=RCIM_Cubic, tangentMode=RCTM_User, time=1, value=0}
        ]}}})
    Scale RGB=&MakeVectorFromLinearColorRGB(Color=&ColorFromCurve(ColorCurve=DataInterface() {propertyValues {
        AlphaCurve {keys []}
        BlueCurve {keys [<at 0.0912 1 RCIM_Cubic>,<at 0.7115 0.8>]}
        GreenCurve {keys [<at 0.0912 1 RCIM_Cubic>,<at 0.7115 0.8>]}
        RedCurve {keys [<at 0.0912 1 RCIM_Cubic>,<at 0.7115 0.8>]}
       }}))
    ScaleA=true
    ScaleRGB=true
   )
   &ScaleColor(
    Scale Alpha=&Lerp_Float(
     A=&FloatFromCurve(FloatCurve=DataInterface() {propertyValues {Curve {keys [
          {arriveTangent=4.1772, interpMode=RCIM_Cubic, leaveTangent=4.1772, tangentMode=RCTM_User, time=0, value=0}
          {arriveTangent=0.0733, interpMode=RCIM_Cubic, leaveTangent=0.0733, tangentMode=RCTM_User, time=0.1408, value=0.9847}
          {arriveTangent=-0.0405, interpMode=RCIM_Cubic, leaveTangent=-0.0405, tangentMode=RCTM_User, time=1, value=0}
         ]}}})
     B=1
     Alpha=&FloatFromCurve(FloatCurve=DataInterface() {propertyValues {Curve {keys [<at 0.18 1>,<at 0.2 0>]}}}, CurveIndex=$Emitter.Age)
    )
    ScaleA=true
    ScaleRGB=false
   ) {module=ScaleColor001}
   &ScaleRibbonWidth(Ribbon Width Scale=&FloatFromCurve(FloatCurve=DataInterface() {propertyValues {Curve {keys [{arriveTangent=-1.8267, interpMode=RCIM_Cubic, leaveTangent=-1.8267, tangentMode=RCTM_User, time=0.2, value=1}, <at 1 0.5>]}}}))
  ]
  rendererClasses [&NiagaraRibbonRendererProperties]
  renderers [{rendererClass=&NiagaraRibbonRendererProperties}]
 }}
summary {
 systemName=NS_Pickaxe_WidePlane_Trail
 emitters [{emitterName=MainTrail, rendererClasses [&NiagaraRibbonRendererProperties]}]
 userVariables [
  User.bisSwinging {type=NiagaraBool}
  User.RibbonIDOffset {type=NiagaraInt32}
 ]
}
